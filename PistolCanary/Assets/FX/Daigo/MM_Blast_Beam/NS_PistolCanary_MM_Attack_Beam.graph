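# Niagara system for the PistolCanary MM attack beam, a glowing blast from the emitter toward User.BeamEnd_0.
# Beams lays six spawn-group ribbon layers along a Hermite spline, with masked widths, waveform flicker and loop-age tapering.
# Sine-jittered energy ribbons, muzzle and core flares, streaks along System.BeamDirection, energy meshes and an impact flare with light surround it.
# System.TimeForDis drives a shared dissipation that fades flares and throttles streak spawning.
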NiagaraSystem /PistolCanary/Assets/FX/Daigo/MM_Blast_Beam/NS_PistolCanary_MM_Attack_Beam
dependencies {
 usedDataInterfaces [&NiagaraDataInterfaceCurve, &NiagaraDataInterfaceColorCurve, &NiagaraDataInterfaceCamera]
 usedDynamicInputs [
  &TransformPosition
  &VectorLength
  &Subtract_Position
  &RemapRange
  &ReturnCameraPosition
  &SetBoolByFloatComparison
  &Subtract_Float
  &MakeCustomFloatFromBool
  &DistanceBetweenPositions
 ]
 usedModules [
  &SystemState
  &Timeline
  &InterpolateOverTime
  &self:SystemSpawnScript.SystemScriptSource.SystemScriptGraph.NiagaraNodeAssignment_0.SetVariables_#1
  &self:SystemSpawnScript.SystemScriptSource.SystemScriptGraph.NiagaraNodeAssignment_10.SetVariables_#2
  &self:SystemSpawnScript.SystemScriptSource.SystemScriptGraph.NiagaraNodeAssignment_2.SetVariables_#3
  &self:SystemSpawnScript.SystemScriptSource.SystemScriptGraph.NiagaraNodeAssignment_3.SetVariables_#4
  &self:SystemSpawnScript.SystemScriptSource.SystemScriptGraph.NiagaraNodeAssignment_4.SetVariables_#5
  &self:SystemSpawnScript.SystemScriptSource.SystemScriptGraph.NiagaraNodeAssignment_5.SetVariables_#6
  &self:SystemSpawnScript.SystemScriptSource.SystemScriptGraph.NiagaraNodeAssignment_6.SetVariables_#7
  &self:SystemSpawnScript.SystemScriptSource.SystemScriptGraph.NiagaraNodeAssignment_1.SetVariables_#8
  &self:SystemSpawnScript.SystemScriptSource.SystemScriptGraph.NiagaraNodeAssignment_7.SetVariables_#9
  &self:SystemSpawnScript.SystemScriptSource.SystemScriptGraph.NiagaraNodeAssignment_8.SetVariables_#10
  &self:SystemSpawnScript.SystemScriptSource.SystemScriptGraph.NiagaraNodeAssignment_9.SetVariables_#11
  &DoOnce
  &self:SystemSpawnScript.SystemScriptSource.SystemScriptGraph.NiagaraNodeAssignment_12.SetVariables_#12
 ]
}
emitters {
 Beams {
  emitterUpdateScript [
   &EmitterState()
   &SpawnBurst_Instantaneous(Spawn Count=20, Use Spawn Probability=false, UseLoopCountLimit=false)
   &SpawnBurst_Instantaneous(Spawn Count=20, Spawn Group=1, Use Spawn Probability=false, UseLoopCountLimit=false) {module=SpawnBurst_Instantaneous001}
   &SpawnBurst_Instantaneous(Spawn Count=20, Spawn Group=2, Use Spawn Probability=false, UseLoopCountLimit=false) {module=SpawnBurst_Instantaneous002}
   &SpawnBurst_Instantaneous(Spawn Count=20, Spawn Group=3, Use Spawn Probability=false, UseLoopCountLimit=false) {module=SpawnBurst_Instantaneous003}
   &SpawnBurst_Instantaneous(Spawn Count=20, Spawn Group=4, Use Spawn Probability=false, UseLoopCountLimit=false) {module=SpawnBurst_Instantaneous004}
   &SpawnBurst_Instantaneous(Spawn Count=20, Spawn Group=5, Use Spawn Probability=false, UseLoopCountLimit=false) {module=SpawnBurst_Instantaneous005}
   /SetVariables(Emitter.ScaleW:NiagaraFloat=&Waveform(Global Amplitude Scale=0.5, "Waveform [1]"=::NewEnumerator12, "[1] Frequency"=20, "[1] Amplitude Min/Max"=<0 1>)) {isSetParameters=true}
   /SetVariables(Emitter.BeamWidthScale:NiagaraFloat=&FloatFromCurve(
     FloatCurve=DataInterface() {propertyValues {Curve {
        defaultValue=1
        keys [
         {tangentMode=RCTM_User, time=0, value=1}
         {arriveTangent=-0.5858, interpMode=RCIM_Cubic, leaveTangent=-0.5858, tangentMode=RCTM_User, time=0.5, value=1}
         {arriveTangent=-5.801, interpMode=RCIM_Cubic, leaveTangent=-5.801, tangentMode=RCTM_User, time=1, value=0}
        ]
       }}}
     CurveIndex=$Emitter.NormalizedLoopAge
     Scale Curve=2.5
    )) {isSetParameters=true}
  ]
  particleSpawnScript [
   /SetVariables(Particles.RibbonID=/Niagara/DynamicInputs/Meshes/MakeRibbonIDFromSpawnGroup()) {isSetParameters=true}
   &PartitionParticles(
    Number Of Partitions=1
    Particles Per Partition=20
    Partition Mode:ENiagara_PartitionMode=::NewEnumerator1
    Write Partition Index to Ribbon Link Order:ENiagara_AttributeSamplingApplyOutput=::NewEnumerator0
   )
   &InitializeParticle(
    Color=&MergeHelper_ColorSwitch(
     0=<rgba 0 0 0 0>
     1=<rgba 0.6 0.03 0.0775 1.5>
     2=<rgba 0 0 0 0>
     3=<rgba 0.05 0.0005 0.0046 0.1>
     4=<rgba 0 0 0 0>
     5=<rgba 0 0 0 0>
     Index=&SpawnGroup()
     Num=::NewEnumerator4
    )
    Lifetime=$System.CurrentLoopDuration
    Ribbon Width=&MergeHelper_FloatSwitch(0=150, 1=100, 2=175, 3=700.3717, 4=200, 5=100, Index=&SpawnGroup(), Num=::NewEnumerator4)
    Ribbon Width Mode=::NewEnumerator1
    UsePositionOffset=false
    Write Material Random=true
   )
  ]
  particleUpdateScript [
   &ParticleState()
   /Game/Effects/Niagara/Modules/Curves/FN_HermiteSpline(
    Position Blend:NiagaraFloat=$Particles.Initial.PartitionParticles.Normalized Index In Partition
    Position End:NiagaraPosition=$User.BeamEnd_0
    Position Start:NiagaraPosition=$Engine.Emitter.SimulationPosition
    Write Particles Position=true
   )
   /SetVariables(
    Particles.Beam Fade_Width:NiagaraFloat=&Multiply_Float(
     A=&Multiply_Float(A=$Particles.Initial.RibbonWidth, B=$Emitter.BeamWidthScale)
     B=&MaskFloatBySpawnGroup(Masked Float=1, Passthrough Float=0, Spawn Group Mask=2)
    )
    Particles.BeamDetail_Width:NiagaraFloat=&Multiply_Float(
     A=&Multiply_Float(A=$Particles.Initial.RibbonWidth, B=$Emitter.BeamWidthScale)
     B=&MaskFloatBySpawnGroup(
      Masked Float=/Game/Effects/Niagara/DynamicInputs/Math/FN_Fit(
       Clamp:NiagaraBool=false
       Value:NiagaraFloat=$Emitter.ScaleW
       SourceMin:NiagaraFloat=0
       SourceMax:NiagaraFloat=1
       DestMin:NiagaraFloat=0.5
       DestMax:NiagaraFloat=0.45
      )
      Passthrough Float=0
      Spawn Group Mask=4
     )
    )
    Particles.BeamGlow2_Width:NiagaraFloat=&Multiply_Float(
     A=&Multiply_Float(A=$Particles.Initial.RibbonWidth, B=$Emitter.BeamWidthScale)
     B=&MaskFloatBySpawnGroup(
      Masked Float=&FN_Fit(
       Clamp:NiagaraBool=false
       Value:NiagaraFloat=$Emitter.ScaleW
       SourceMin:NiagaraFloat=0
       SourceMax:NiagaraFloat=1
       DestMin:NiagaraFloat=0.25
       DestMax:NiagaraFloat=1
      )
      Passthrough Float=0
      Spawn Group Mask=3
     )
    )
    Particles.BeamGlow_Width:NiagaraFloat=&Multiply_Float(
     A=&Multiply_Float(A=$Particles.Initial.RibbonWidth, B=$Emitter.BeamWidthScale)
     B=&MaskFloatBySpawnGroup(Masked Float=0.25, Passthrough Float=0, Spawn Group Mask=1)
    )
    Particles.BeamParticle_Width:NiagaraFloat=&Multiply_Float(
     A=&Multiply_Float(A=$Particles.Initial.RibbonWidth, B=$Emitter.BeamWidthScale)
     B=&MaskFloatBySpawnGroup(Masked Float=1, Passthrough Float=0, Spawn Group Mask=5)
    )
    Particles.ProjectileBeam_Width:NiagaraFloat=&Multiply_Float(
     A=&Multiply_Float(A=$Particles.Initial.RibbonWidth, B=$Emitter.BeamWidthScale)
     B=&MaskFloatBySpawnGroup(Masked Float=1, Passthrough Float=0)
    )
   ) {isSetParameters=true}
   &ScaleColor(
    Scale Alpha=&MergeHelper_FloatSwitch(
     0=1
     1=5
     2=1
     3=&FloatFromCurve(
      FloatCurve=DataInterface() {propertyValues {Curve {keys [
          {interpMode=RCIM_Cubic, tangentMode=RCTM_Break, time=0, value=1}
          {arriveTangent=-3.624, interpMode=RCIM_Cubic, leaveTangent=-0.5922, tangentMode=RCTM_Break, time=0.7, value=0.3}
          {interpMode=RCIM_Cubic, tangentMode=RCTM_Break, time=1, value=0.4}
         ]}}}
      CurveIndex=$Particles.RibbonLinkOrder
     )
     4=1
     5=1
     Index=&SpawnGroup()
     Num=::NewEnumerator4
    )
    Scale RGB=<2 2 10>
    ScaleA=true
    ScaleRGB=true
   )
   &DynamicMaterialParameters(
    Index 0 Param 1=&MaskFloatBySpawnGroup(Masked Float=2, Passthrough Float=1)
    Index 2 Param 1=&Lerp_Float(
     A=-0.8775
     B=1
     Alpha=&Multiply_Float(
      A=&FloatFromCurve(FloatCurve=DataInterface() {propertyValues {Curve {keys [<at 0 0>,<at 0.1 1>]}}}, CurveIndex=$Engine.System.Age)
      B=1
     )
    )
    Index 3 Param 1=&FN_Fit(
     Clamp:NiagaraBool=false
     Value:NiagaraFloat=$Emitter.ScaleW
     SourceMin:NiagaraFloat=0
     SourceMax:NiagaraFloat=1
     DestMin:NiagaraFloat=0
     DestMax:NiagaraFloat=&FloatFromCurve(
      FloatCurve=DataInterface() {propertyValues {Curve {keys [{interpMode=RCIM_Cubic, tangentMode=RCTM_Break, time=0, value=0}, <at 0.5 1>]}}}
      CurveIndex=$Engine.System.Age
      Scale Curve=&FloatFromCurve(
       FloatCurve=DataInterface() {propertyValues {Curve {keys [{interpMode=RCIM_Cubic, leaveTangent=-3.7341, tangentMode=RCTM_Break, time=0, value=1}, <at 1 -1>]}}}
       CurveIndex=$System.TimeForDis
      )
     )
    )
    Module.Index0Param0Write=true
    Module.Index0Param1Write=true
    Module.Index0Param2Write=false
    Module.Index0Param3Write=false
    Module.Index2Param0Write=true
    Module.Index2Param1Write=true
    Module.Index2Param2Write=true
    Module.Index2Param3Write=true
    Module.Index3Param0Write=true
    Module.Index3Param1Write=true
    Module.Index3Param2Write=true
    Module.Index3Param3Write=true
    Write Parameter Index 2=true
    Write Parameter Index 3=true
   )
   &ScaleColor(
    Scale Alpha=&FloatFromCurve(
     FloatCurve=DataInterface() {propertyValues {Curve {
        defaultValue=1
        keys [
         {tangentMode=RCTM_User, time=0, value=1}
         {arriveTangent=-0.0405, interpMode=RCIM_Cubic, leaveTangent=-0.0405, tangentMode=RCTM_User, time=0.7009, value=1}
         {arriveTangent=-8.0089, interpMode=RCIM_Cubic, leaveTangent=-8.0089, tangentMode=RCTM_User, time=1, value=0}
        ]
       }}}
     CurveIndex=$Emitter.NormalizedLoopAge
    )
    Scale RGB=<100 100 100>
    ScaleA=true
    ScaleRGB=true
   ) {module=ScaleColor001}
  ]
  rendererClasses [&NiagaraRibbonRendererProperties]
  renderers [
   {rendererClass=&NiagaraRibbonRendererProperties}
   {rendererClass=&NiagaraRibbonRendererProperties, rendererIndex=1}
   {rendererClass=&NiagaraRibbonRendererProperties, rendererIndex=2}
   {rendererClass=&NiagaraRibbonRendererProperties, rendererIndex=3}
  ]
 }
 CoreFlare001 {
  emitterSpawnScript [/SetVariables(Emitter.SizeCurve=DataInterface() {propertyValues {Curve {keys [
        <at 0.2 0 RCIM_Cubic>
        {arriveTangent=2.9237, interpMode=RCIM_Cubic, leaveTangent=1.25, tangentMode=RCTM_Break, time=0.8, value=0.95}
        {arriveTangent=-0.193, interpMode=RCIM_Cubic, leaveTangent=-0.193, tangentMode=RCTM_User, time=2, value=0}
       ]}}}) {isSetParameters=true}]
  emitterUpdateScript [
   &EmitterState()
   &SpawnBurst_Instantaneous(Use Spawn Probability=false, UseLoopCountLimit=true)
  ]
  particleSpawnScript [&InitializeParticle(
    Color=<rgba 0.35 0.0379 0.2602 5>
    Lifetime=$System.CurrentLoopDuration
    Position Offset=<0 100 50>
    Sprite Size Mode=::NewEnumerator3
    Uniform Sprite Size=5000
    UsePositionOffset=true
    Write Material Random=true
   )]
  particleUpdateScript [
   &ParticleState()
   &CameraOffset(Camera Offset Amount=400)
   &ScaleColor(
    Scale Alpha=&Waveform(Global Amplitude Scale=0.2, "Waveform [1]"=::NewEnumerator12, "[1] Phase Scale"=10, "[1] Frequency"=100, "[1] Amplitude Min/Max"=<0.5 1>)
    ScaleA=true
    ScaleRGB=true
   )
   &ScaleColor(
    Scale Alpha=&FloatFromCurve(
     FloatCurve=DataInterface() {propertyValues {Curve {keys [{interpMode=RCIM_Cubic, tangentMode=RCTM_Break, time=0, value=0}, <at 0.3 1>]}}}
     CurveIndex=$Engine.System.Age
    )
    ScaleA=true
    ScaleRGB=true
   ) {module=ScaleColor001}
   &ScaleColor(
    Scale Alpha=&FloatFromCurve(
     FloatCurve=DataInterface() {propertyValues {Curve {keys [{interpMode=RCIM_Cubic, leaveTangent=-0.1053, tangentMode=RCTM_Break, time=0, value=1}, <at 1 0>]}}}
     CurveIndex=$System.TimeForDis
    )
    ScaleA=true
    ScaleRGB=true
   ) {module=ScaleColor002}
  ]
  rendererClasses [&NiagaraSpriteRendererProperties]
  renderers [{rendererClass=&NiagaraSpriteRendererProperties}]
 }
 Energy_Sphere001 {
  emitterUpdateScript [
   &EmitterState()
   &SpawnRate(SpawnRate=4, Use Spawn Probability=false) {enabled=false}
   &SpawnBurst_Instantaneous(Use Spawn Probability=false, UseLoopCountLimit=true)
  ]
  particleSpawnScript [
   &InitializeParticle(
    Color=<rgba 0.2697 0.2802 0.5 1>
    Lifetime=$User.Duration
    Mesh Scale=<10 10 10>
    Mesh Scale Mode=::NewEnumerator4
    Position Offset=<0 450 185>
    Sprite Size Mode=::NewEnumerator3
    Uniform Sprite Size=14500
    UsePositionOffset=true
    Write Material Random=true
   ) {module=InitializeParticle001}
   &InitialMeshOrientation(
    Mesh Axis To Orient=<0 0 -1>
    Mesh Orientation Mode=::NewEnumerator2
    Orientation Vector=&NormalizeVector(
     Vector To Normalize=&SubtractVectorFromPosition(Position=$User.BeamEnd_0, Vector=Linked() {linkedVariable {name=Engine.Owner.Position, type=NiagaraPosition}})
     Safe Normalize=true
     Safe Fallback Threshold=0
     Safe Fallback Vector=<0 0 0.001>
    )
    Use Rotation Vector=false
   )
  ]
  particleUpdateScript [
   &ParticleState()
   &ScaleSpriteSize()
   &ScaleColor(
    Scale Alpha=&FloatFromCurve(
     FloatCurve=DataInterface() {propertyValues {Curve {defaultValue=1, keys [{tangentMode=RCTM_User, time=0, value=1}, {tangentMode=RCTM_User, time=1, value=0}]}}}
     Scale Curve=2
    )
    ScaleA=true
    ScaleRGB=true
   )
   &ScaleMeshSize(Scale Factor=&VectorFromFloat(Value=&FloatFromCurve(
      FloatCurve=DataInterface() {propertyValues {Curve {keys [
          {arriveTangentWeight=0.0511, interpMode=RCIM_Cubic, leaveTangentWeight=0.0511, tangentMode=RCTM_User, time=0, value=1.04}
          <at 1 2 RCIM_Cubic>
         ]}}}
      Scale Curve=3
     ))) {enabled=false}
   &DynamicMaterialParameters(
    Index 0 Param 1=1
    Index 0 Param 2=1
    Module.Index0Param0Write=true
    Module.Index0Param1Write=true
    Module.Index0Param2Write=true
    Module.Index0Param3Write=true
   )
   &FN_Sprite_Attributes(
    Color=&MultiplyLinearColor(
     Linear Color=$Particles.Color
     Scale Factor=&MakeVector4_FromLinearColor(LinearColor=<rgba 0.1765 0.0373 0.4 0.4>)
    )
    Create Color Attribute=true
    Write to Module Namespace=true
   )
  ]
  rendererClasses [&NiagaraMeshRendererProperties]
  renderers [{rendererClass=&NiagaraMeshRendererProperties, rendererIndex=0}]
 }
 Impact_Flare {
  emitterUpdateScript [
   &EmitterState(Life Cycle Mode=::NewEnumerator1, Loop Behavior=::NewEnumerator1, Loop Duration=1, UseLoopDelay=false)
   &SpawnBurst_Instantaneous(Use Spawn Probability=false, UseLoopCountLimit=false)
  ]
  particleSpawnScript [
   &InitializeParticle(
    Color=<rgba 2 2 2 1>
    Lifetime=0.7
    Position Offset Coordinate Space=::Local
    Sprite Size Mode=::NewEnumerator3
    Uniform Sprite Size=4500
    UsePositionOffset=true
    Write Material Random=true
   )
   &DynamicMaterialParameters(
    Index 0 Param 1=1
    Index 0 Param 2=1
    Index 0 Param 3=1
    Index 0 Param 4=1
    Module.Param0WriteEnabled=true
    Module.Param1WriteEnabled=true
    Module.Param2WriteEnabled=true
    Module.Param3WriteEnabled=true
   )
   &CameraOffset(Camera Offset Amount=300)
  ]
  particleUpdateScript [
   &ParticleState()
   &ScaleColor(
    Scale Alpha=&FloatFromCurve(FloatCurve=DataInterface() {propertyValues {Curve {keys [
         {arriveTangent=6.0888, interpMode=RCIM_Cubic, leaveTangent=6.0888, tangentMode=RCTM_User, time=0, value=0}
         <at 0.4 1 RCIM_Cubic>
         <at 1 0 RCIM_Cubic>
        ]}}})
    Scale RGB=&MakeVectorFromLinearColorRGB(Color=&ColorFromCurve(ColorCurve=DataInterface() {propertyValues {
        AlphaCurve {keys []}
        BlueCurve {keys [<at 0 1.1083 RCIM_Cubic>,<at 0.2799 0.111>,<at 1 0.0157 RCIM_Cubic>]}
        GreenCurve {keys [<at 0 0.3 RCIM_Cubic>,<at 0.2799 0.0301>,<at 1 0 RCIM_Cubic>]}
        RedCurve {keys [<at 0 10 RCIM_Cubic>,<at 0.2799 1>,<at 1 0.1878 RCIM_Cubic>]}
       }}))
    ScaleA=true
    ScaleRGB=true
   )
   &ScaleColor(ScaleA=true, ScaleRGB=true) {module=ScaleColor001}
   &FN_Sprite_Attributes(
    Color=&Multiply_LinearColorByFloat(Linear Color=$Particles.Color, Float=2)
    Create Color Attribute=true
    Create SpriteSize Attribute=true
    SpriteSize=&Multiply_Vector2D(A=$Particles.SpriteSize, B=<1 0.4232>)
    Write to Module Namespace=true
   )
   &FN_Light_Attributes(
    Color=<rgba 1 0.0387 0.8863 1>
    Create Color Attribute=true
    Create Position Attribute=true
    Create Radius Attribute=true
    Position=&AddVectorToPosition(Position=$User.BeamEnd_0, Vector=<0 2200 500>)
    Radius=&FloatFromCurve(
     FloatCurve=DataInterface() {propertyValues {Curve {
        defaultValue=1
        keys [{interpMode=RCIM_Cubic, tangentMode=RCTM_User, time=0.3, value=1}, {interpMode=RCIM_Cubic, tangentMode=RCTM_User, time=1, value=0}]
       }}}
     Scale Curve=4000
    )
    Scale Color=&FloatFromCurve(
     FloatCurve=DataInterface() {propertyValues {Curve {
        defaultValue=1
        keys [{interpMode=RCIM_Cubic, tangentMode=RCTM_User, time=0, value=1}, {interpMode=RCIM_Cubic, tangentMode=RCTM_User, time=1, value=0}]
       }}}
     Scale Curve=100
    )
   )
  ]
  rendererClasses [&NiagaraSpriteRendererProperties, &NiagaraLightRendererProperties]
  renderers [
   {rendererClass=&NiagaraSpriteRendererProperties}
   {rendererClass=&NiagaraSpriteRendererProperties, rendererIndex=1}
   {rendererClass=&NiagaraLightRendererProperties, rendererIndex=2}
  ]
 }
 Minimal {
  emitterUpdateScript [
   &EmitterState()
   &SpawnRate(
    SpawnRate=&FloatFromCurve(
     FloatCurve=DataInterface() {propertyValues {Curve {keys [
         {interpMode=RCIM_Cubic, leaveTangent=0.3372, tangentMode=RCTM_Break, time=3.9, value=0}
         {arriveTangent=1.6707, tangentMode=RCTM_User, time=4, value=1}
         <at 4.1 0>
        ]}}}
     CurveIndex=$Emitter.LoopedAge
     Scale Curve=500
    )
    Use Spawn Probability=false
   ) {enabled=false, module=SpawnRate001}
   &SpawnRate(
    SpawnRate=&Multiply_Float(
     A=100
     B=&FloatFromCurve(
      FloatCurve=DataInterface() {propertyValues {Curve {keys [{interpMode=RCIM_Cubic, leaveTangent=-3.7201, tangentMode=RCTM_Break, time=0, value=1}, <at 1 0>]}}}
      CurveIndex=$System.TimeForDis
     )
    )
    Use Spawn Probability=false
   )
  ]
  particleSpawnScript [
   &InitializeParticle(
    Color=<rgba 10 0 0.8333 0.5>
    Lifetime=&RandomRangeFloat(Minimum=0.2, Maximum=0.3)
    Sprite Size=&Multiply_Vector2D_ByFloat(Float=&RandomRangeFloat(Minimum=1, Maximum=2), Vector2D=<10 120>)
    Sprite Size Mode=::NewEnumerator4
    UsePositionOffset=false
    Write Material Random=true
   )
   &ShapeLocation(MaskSpawnGroup=false, Override Seed=false, Sphere Radius=10, UseOwnerScale=true)
   &AddVelocity(Module.Use Falloff Distance=false, Override Seed=false, Velocity Mode=::NewEnumerator1, Velocity Speed=3318.3464)
   &AddVelocity(Rotation Coordinate Space=::World, Velocity=$System.BeamDirection, Velocity Speed Scale=3265.6787) {module=AddVelocity001}
  ]
  particleUpdateScript [
   &ParticleState()
   &ScaleSpriteSize(
    Scale Factor=&MakeVector2D(
     X=&FloatFromCurve(FloatCurve=DataInterface() {propertyValues {Curve {keys [
          {interpMode=RCIM_Cubic, leaveTangent=11.7691, tangentMode=RCTM_Break, time=0, value=0}
          {arriveTangent=0.0629, interpMode=RCIM_Cubic, leaveTangent=0.0629, tangentMode=RCTM_User, time=0.2, value=1}
          {interpMode=RCIM_Cubic, leaveTangent=0.027, tangentMode=RCTM_Break, time=0.35, value=1}
          {arriveTangent=0.0384, tangentMode=RCTM_User, time=1, value=0}
         ]}}})
     Y=1
    )
    Scale Sprite Size Mode=::NewEnumerator2
   )
   &CurlNoiseForce(Bias Noise=false, Noise Frequency=10, Noise Strength=550.2112, Pan Noise=false) {enabled=false}
   &SolveForcesAndVelocity(Clamp Velocity=false, Limit Acceleration=false)
   &ScaleSpriteSizeBySpeed(Max Scale Factor=<1 2>, Velocity Threshold=3500)
   &FN_Sprite_Attributes(
    Color=&MultiplyLinearColor(Linear Color=$Particles.Color, Scale Factor=<0.025 0.025 0.025 1>)
    Create Color Attribute=true
    Create SpriteSize Attribute=true
    SpriteSize=&Multiply_Vector2D(A=$Particles.SpriteSize, B=<5 3>)
    Write to Module Namespace=true
   )
   &FN_Sprite_Attributes(
    Color=&MultiplyLinearColor(Linear Color=<rgba 7 0.35 0.9042 1>, Scale Factor=<0.1 0.1 0.1 0.5>)
    Create Color Attribute=true
    Create SpriteSize Attribute=true
    SpriteSize=&Multiply_Vector2D(A=$Particles.SpriteSize, B=<10 0.5>)
    Write to Module Namespace=true
   ) {enabled=false, module=FN_Sprite_Attributes001}
   &ScaleColor(
    Scale Alpha=&FloatFromCurve(FloatCurve=DataInterface() {propertyValues {Curve {
        defaultValue=1
        keys [
         {arriveTangent=9.0725, interpMode=RCIM_Cubic, leaveTangent=9.0725, tangentMode=RCTM_User, time=0, value=0}
         {arriveTangent=-2.5249, interpMode=RCIM_Cubic, leaveTangent=-2.5249, tangentMode=RCTM_User, time=0.1408, value=0.9847}
         {arriveTangent=-0.0405, interpMode=RCIM_Cubic, leaveTangent=-0.0405, tangentMode=RCTM_User, time=1, value=0}
        ]
       }}})
    ScaleA=true
    ScaleRGB=true
   )
  ]
  rendererClasses [&NiagaraSpriteRendererProperties]
  renderers [
   {rendererClass=&NiagaraSpriteRendererProperties}
   {rendererClass=&NiagaraSpriteRendererProperties, rendererIndex=1}
  ]
 }
 Minimal002 {
  emitterUpdateScript [
   &EmitterState()
   &SpawnRate(
    SpawnRate=&FloatFromCurve(
     FloatCurve=DataInterface() {propertyValues {Curve {keys [
         {interpMode=RCIM_Cubic, leaveTangent=0.3372, tangentMode=RCTM_Break, time=3.9, value=0}
         {arriveTangent=1.6707, tangentMode=RCTM_User, time=4, value=1}
         <at 4.1 0>
        ]}}}
     CurveIndex=$Emitter.LoopedAge
     Scale Curve=500
    )
    Use Spawn Probability=false
   ) {enabled=false, module=SpawnRate001}
   &SpawnRate(
    SpawnRate=&Multiply_Float(
     A=100
     B=&FloatFromCurve(
      FloatCurve=DataInterface() {propertyValues {Curve {keys [{interpMode=RCIM_Cubic, leaveTangent=-3.7201, tangentMode=RCTM_Break, time=0, value=1}, <at 1 0>]}}}
      CurveIndex=$System.TimeForDis
     )
    )
    Use Spawn Probability=false
   )
  ]
  particleSpawnScript [
   &InitializeParticle(
    Color=<rgba 0.05 0 0.0042 5>
    Lifetime=&RandomRangeFloat(Minimum=0.2, Maximum=0.4)
    Sprite Size=&Multiply_Vector2D_ByFloat(Float=&RandomRangeFloat(Minimum=1, Maximum=2), Vector2D=<20 100>)
    Sprite Size Mode=::NewEnumerator4
    UsePositionOffset=false
    Write Material Random=true
   )
   &ShapeLocation(MaskSpawnGroup=false, Override Seed=false, Sphere Radius=10, UseOwnerScale=true)
   &AddVelocity(Module.Use Falloff Distance=false, Override Seed=false, Velocity Mode=::NewEnumerator1, Velocity Speed=2200)
   &AddVelocity(Rotation Coordinate Space=::World, Velocity=$System.BeamDirection, Velocity Speed Scale=1741.5543) {module=AddVelocity001}
  ]
  particleUpdateScript [
   &ParticleState()
   &ScaleSpriteSize(
    Scale Factor=&MakeVector2D(
     X=&FloatFromCurve(FloatCurve=DataInterface() {propertyValues {Curve {keys [
          {interpMode=RCIM_Cubic, leaveTangent=11.7691, tangentMode=RCTM_Break, time=0, value=0}
          {arriveTangent=0.0629, interpMode=RCIM_Cubic, leaveTangent=0.0629, tangentMode=RCTM_User, time=0.2, value=1}
          {interpMode=RCIM_Cubic, leaveTangent=0.027, tangentMode=RCTM_Break, time=0.35, value=1}
          {arriveTangent=0.0384, tangentMode=RCTM_User, time=1, value=0}
         ]}}})
     Y=1
    )
    Scale Sprite Size Mode=::NewEnumerator2
   )
   &CurlNoiseForce(Bias Noise=false, Noise Frequency=10, Noise Strength=550.2112, Pan Noise=false) {enabled=false}
   &SolveForcesAndVelocity(Clamp Velocity=false, Limit Acceleration=false)
   &ScaleSpriteSizeBySpeed(Max Scale Factor=<1 2>, Velocity Threshold=3500)
   &FN_Sprite_Attributes(
    Color=&MultiplyLinearColor(Linear Color=$Particles.Color, Scale Factor=<0.5 0.5 0.5 0.3>)
    Create Color Attribute=true
    Create SpriteSize Attribute=true
    SpriteSize=&Multiply_Vector2D(A=$Particles.SpriteSize, B=<3 1>)
    Write to Module Namespace=true
   )
   &FN_Sprite_Attributes(
    Color=&MultiplyLinearColor(Linear Color=<rgba 7 0.35 0.9042 1>, Scale Factor=<0.1 0.1 0.1 0.5>)
    Create Color Attribute=true
    Create SpriteSize Attribute=true
    SpriteSize=&Multiply_Vector2D(A=$Particles.SpriteSize, B=<10 0.5>)
    Write to Module Namespace=true
   ) {enabled=false, module=FN_Sprite_Attributes001}
   &ScaleColor(
    Scale Alpha=&FloatFromCurve(FloatCurve=DataInterface() {propertyValues {Curve {
        defaultValue=1
        keys [
         {arriveTangent=9.0725, interpMode=RCIM_Cubic, leaveTangent=9.0725, tangentMode=RCTM_User, time=0, value=0}
         {arriveTangent=-2.5249, interpMode=RCIM_Cubic, leaveTangent=-2.5249, tangentMode=RCTM_User, time=0.1408, value=0.9847}
         {arriveTangent=-0.0405, interpMode=RCIM_Cubic, leaveTangent=-0.0405, tangentMode=RCTM_User, time=1, value=0}
        ]
       }}})
    ScaleA=true
    ScaleRGB=true
   )
  ]
  rendererClasses [&NiagaraSpriteRendererProperties]
  renderers [{rendererClass=&NiagaraSpriteRendererProperties}]
 }
 Muzzle_Inner_Flares {
  emitterUpdateScript [
   &EmitterState()
   &SpawnBurst_Instantaneous(Use Spawn Probability=false, UseLoopCountLimit=false)
   &SpawnRate(SpawnRate=40, Use Spawn Probability=false)
  ]
  particleSpawnScript [
   &InitializeParticle(
    Color=$User.PrimaryColor
    Lifetime Max=0.175
    Lifetime Min=0.075
    Lifetime Mode=::NewEnumerator1
    Sprite Size Mode=::NewEnumerator1
    Uniform Sprite Size Max=1000
    Uniform Sprite Size Min=600
    UsePositionOffset=false
    Write Material Random=true
   )
   &CameraOffset(Camera Offset Amount=400)
   /SetVariables(
    Particles.Animorphic Color:LinearColor=<rgba 10 0.3013 1.1095 0.05>
    Particles.Animorphic Size:Vector2f=<1733.7466 20.5981>
    Particles.Soft Glow Size:Vector2f=<3289.3108 195.8096>
   ) {isSetParameters=true}
  ]
  particleUpdateScript [
   &ParticleState()
   &ScaleColor(
    Scale Alpha=&FloatFromCurve(FloatCurve=DataInterface() {propertyValues {Curve {keys [
         {interpMode=RCIM_Cubic, tangentMode=RCTM_User, time=0, value=0}
         {interpMode=RCIM_Cubic, tangentMode=RCTM_User, time=0.5, value=1}
         {tangentMode=RCTM_User, time=1, value=0}
        ]}}})
    ScaleA=true
    ScaleRGB=true
   )
   /SetVariables(Particles.Soft Glow Color:LinearColor=&MakeLinearColorFromVectorAndFloat(
     "Vector (RGB)"=&MakeVectorFromLinearColorRGB(Color=<rgba 10 0 0.8333 1>)
     "Float (Alpha)"=&FloatFromCurve(
      FloatCurve=DataInterface() {propertyValues {Curve {keys [
          {interpMode=RCIM_Cubic, tangentMode=RCTM_User, time=0, value=0}
          {interpMode=RCIM_Cubic, tangentMode=RCTM_User, time=0.5, value=1}
          {tangentMode=RCTM_User, time=1, value=0}
         ]}}}
      Scale Curve=&Multiply_Float(
       A=0.005
       B=&FloatFromCurve(
        FloatCurve=DataInterface() {propertyValues {Curve {keys [
            {interpMode=RCIM_Cubic, leaveTangent=-2.0566, tangentMode=RCTM_Break, time=0, value=1}
            {arriveTangent=-0.9453, tangentMode=RCTM_User, time=1, value=0}
           ]}}}
        CurveIndex=$System.TimeForDis
       )
      )
     )
    )) {isSetParameters=true}
   /SetVariables(
    Particles.Corona Color:LinearColor=&MakeLinearColorFromVectorAndFloat(
     "Vector (RGB)"=&MakeVectorFromLinearColorRGB(Color=<rgba 1 0.0889 0.1786 1>)
     "Float (Alpha)"=&FloatFromCurve(
      FloatCurve=DataInterface() {propertyValues {Curve {keys [
          <at 0 0 RCIM_Cubic>
          {arriveTangent=2.5, interpMode=RCIM_Cubic, leaveTangent=2.5, time=0.165, value=0.65}
          {arriveTangent=0.1786, interpMode=RCIM_Cubic, leaveTangent=0.1786, time=0.4, value=1}
          {arriveTangent=-1.6667, interpMode=RCIM_Cubic, leaveTangent=-1.6667, time=0.725, value=0.75}
          <at 1 0 RCIM_Cubic>
         ]}}}
      Scale Curve=&Multiply_Float(
       A=0.5
       B=&FloatFromCurve(
        FloatCurve=DataInterface() {propertyValues {Curve {keys [
            {interpMode=RCIM_Cubic, leaveTangent=-2.0566, tangentMode=RCTM_Break, time=0, value=1}
            {arriveTangent=-0.9453, tangentMode=RCTM_User, time=1, value=0}
           ]}}}
        CurveIndex=$System.TimeForDis
       )
      )
     )
    )
    Particles.Corona Size:Vector2f=&Vector2DFromFloat(Value=&FloatFromCurve(
      FloatCurve=DataInterface() {propertyValues {Curve {keys [<at 0 1>,<at 0.5 0.7>,<at 1 1>]}}}
      Scale Curve=&Waveform(Global Amplitude Scale=2000, "Waveform [1]"=::NewEnumerator12, "[1] Phase Scale"=50, "[1] Frequency"=1, "[1] Amplitude Min/Max"=<0.85 1>)
     ))
   ) {isSetParameters=true}
   &DynamicMaterialParameters(
    Index 3 Param 1=&FloatFromCurve(FloatCurve=DataInterface() {propertyValues {Curve {keys [<at 0 0.7 RCIM_Cubic>,<at 0.5 0.25 RCIM_Cubic>,<at 1 0.7 RCIM_Cubic>]}}})
    Index 3 Param 2=1
    Index 3 Param 3=&RandomRangeFloat(Minimum=0.25, Maximum=0.35)
    Index 3 Param 4=1
    Module.Param0WriteEnabled3=true
    Module.Param1WriteEnabled3=true
    Module.Param2WriteEnabled3=true
    Module.Param3WriteEnabled3=true
    Write Parameter Index 0=false
    Write Parameter Index 3=true
   )
   /SetVariables() {isSetParameters=true}
   &ScaleColor(
    Scale Alpha=&FloatFromCurve(
     FloatCurve=DataInterface() {propertyValues {Curve {keys [
         {interpMode=RCIM_Cubic, leaveTangent=-2.0566, tangentMode=RCTM_Break, time=0, value=1}
         {arriveTangent=-0.9453, tangentMode=RCTM_User, time=1, value=0}
        ]}}}
     CurveIndex=$System.TimeForDis
    )
    ScaleA=true
    ScaleRGB=true
   ) {module=ScaleColor001}
   /SetVariables(Particles.Animorphic Color:LinearColor=&MultiplyLinearColor(
     Linear Color=$Particles.Initial.Animorphic Color
     Scale Factor=&MakeVector4FromVectorAndFloat(
      "Vector (RGB)"=<1 1 1>
      "Float (Alpha)"=&FloatFromCurve(
       FloatCurve=DataInterface() {propertyValues {Curve {keys [{interpMode=RCIM_Cubic, leaveTangent=-3.7201, tangentMode=RCTM_Break, time=0, value=1}, <at 1 0>]}}}
       CurveIndex=$System.TimeForDis
      )
     )
    )) {isSetParameters=true}
  ]
  rendererClasses [&NiagaraSpriteRendererProperties]
  renderers [
   {rendererClass=&NiagaraSpriteRendererProperties}
   {rendererClass=&NiagaraSpriteRendererProperties, rendererIndex=1}
   {rendererClass=&NiagaraSpriteRendererProperties, rendererIndex=2}
   {rendererClass=&NiagaraSpriteRendererProperties, rendererIndex=3}
  ]
 }
 NE_EnergyRibbons {
  emitterSpawnScript [/SetVariables() {isSetParameters=true}]
  emitterUpdateScript [
   /SetVariables(Emitter.CustomTime:NiagaraFloat=&Add_Float(
     A=$Emitter.CustomTime
     B=&Multiply_Float(A=$Engine.DeltaTime, B=&Lerp_Float(A=1, B=3, Alpha=0.5))
    )) {isSetParameters=true}
   &EmitterState(
    Inactive Response=::NewEnumerator1
    Life Cycle Mode=::NewEnumerator1
    Loop Behavior=::NewEnumerator1
    Loop Duration Mode=::NewEnumerator1
    UseLoopDelay=false
   )
   &SpawnBurst_Instantaneous(
    Spawn Count=&Multiply_Int(A=$System.RibbonNumber, B=$System.ParticlePerRibbon)
    Use Spawn Probability=false
    UseLoopCountLimit=false
   )
  ]
  particleSpawnScript [
   &InitializeParticle(
    Lifetime=$System.CurrentLoopDuration
    Ribbon Width=20
    Ribbon Width Mode=::NewEnumerator1
    UsePositionOffset=false
    Write Material Random=true
   )
   &PartitionParticles(
    ID to Partition:NiagaraInt32=&ReturnExecIndex()
    Normalized Index In Partition Includes One=true
    Number Of Partitions=$System.RibbonNumber
    Particles Per Partition:NiagaraInt32=$System.ParticlePerRibbon
    Partition Mode:ENiagara_PartitionMode=::NewEnumerator1
    Write Partition Index to Ribbon Link Order:ENiagara_AttributeSamplingApplyOutput=::NewEnumerator0
    Write Partition to Ribbon ID=::NewEnumerator0
   )
   /SetVariables(Particles.RibbonU0Override=$Particles.PartitionParticles.Normalized Index In Partition) {isSetParameters=true}
  ]
  particleUpdateScript [
   &ParticleState()
   &RotateAroundPoint(
    Radius=2067.3516
    Rotation Center:NiagaraPosition=/Niagara/DynamicInputs/Helpers/SimulationPosition()
    Rotation Rate=5.3597
    Set Particle Position=false
    X Axis:Vector3f=<0 0 0>
    Y Axis:Vector3f=<1 1 0>
   )
   &ScaleRibbonWidth(Ribbon Width Scale=&Lerp_Float(
     A=1.75
     B=7.5
     Alpha=&FloatFromCurve(
      FloatCurve=DataInterface() {propertyValues {Curve {keys [
          {arriveTangent=5.8422, interpMode=RCIM_Cubic, leaveTangent=5.8422, tangentMode=RCTM_User, time=0, value=1}
          {arriveTangent=0.0677, interpMode=RCIM_Cubic, leaveTangent=0.0677, tangentMode=RCTM_User, time=1, value=0.25}
         ]}}}
      CurveIndex=$Particles.PartitionParticles.Normalized Index In Partition
      Scale Curve=&FloatFromCurve(FloatCurve=DataInterface() {propertyValues {Curve {
          defaultValue=1
          keys [
           {interpMode=RCIM_Cubic, tangentMode=RCTM_User, time=0, value=1}
           {arriveTangent=-1, interpMode=RCIM_Cubic, leaveTangent=-1, time=0.7, value=0.825}
           {interpMode=RCIM_Cubic, tangentMode=RCTM_User, time=1, value=0}
          ]
         }}})
     )
    ))
   /SetVariables(Particles.SpiralAlpha:NiagaraFloat=&Add_Float(
     A=&Add_Float(
      A=&Multiply_Float(A=$Emitter.CustomTime, B=0.5)
      B=&Multiply_Float(A=$Particles.PartitionParticles.Normalized Index In Partition, B=0.5)
     )
     B=&Multiply_FloatByInt(
      Float=&Divide_Float(A=1, B=&MakeFloatFromInt(Int=$System.RibbonNumber))
      Integer=$Particles.PartitionParticles.Partition
     )
    )) {enabled=false, isSetParameters=true}
   /SetVariables(Particles.SpiralOffset:Vector3f=&Multiply_VectorByFloat(
     Vector=&MakeVector(
      X=&Cosine(Normalized Angle=$Particles.SpiralAlpha, Period=1, Scale=1, Bias=0)
      Y=0
      Z=&Sine(Normalized Angle=$Particles.SpiralAlpha, Period=1, Scale=1, Bias=0)
     )
     Float=1
    )) {enabled=false, isSetParameters=true}
   /SetVariables(Particles.Position=&AddVectorToPosition(
     Position=&Lerp_Vector(
      A=Linked() {linkedVariable {name=Engine.Owner.Position, type=NiagaraPosition}}
      B=Linked() {linkedVariable {name=User.BeamEnd_0, type=NiagaraPosition}}
      Alpha=$Particles.Initial.PartitionParticles.Normalized Index In Partition
     )
     Vector=&Multiply_VectorByFloat(
      Vector=&RandomRangeVector(Minimum=<-150 -150 -150>, Maximum=<150 150 150>)
      Float=&FloatFromCurve(
       FloatCurve=DataInterface() {propertyValues {Curve {
          defaultValue=1
          keys [
           {tangentMode=RCTM_User, time=0, value=1}
           {arriveTangent=-0.0405, interpMode=RCIM_Cubic, leaveTangent=-0.0405, tangentMode=RCTM_User, time=0.7009, value=1}
           {arriveTangent=-3.8648, interpMode=RCIM_Cubic, leaveTangent=-3.8648, tangentMode=RCTM_User, time=1, value=0.5}
          ]
         }}}
       Scale Curve=1.25
      )
     )
    )) {enabled=false, isSetParameters=true}
   /SetVariables(Particles.Position=&AddVectorToPosition(
     Position=&Lerp_Vector(
      A=Linked() {linkedVariable {name=Engine.Owner.Position, type=NiagaraPosition}}
      B=Linked() {linkedVariable {name=User.BeamEnd_0, type=NiagaraPosition}}
      Alpha=$Particles.Initial.PartitionParticles.Normalized Index In Partition
     )
     Vector=&Multiply_VectorByFloat(
      Vector=&MakeVector(
       X=&Sine(
        Normalized Angle=&Add_Float(A=$Particles.Initial.PartitionParticles.Normalized Index In Partition, B=$Emitter.Age)
        Period=1
        Scale=&RandomRangeFloat(Minimum=-150, Maximum=150)
        Bias=0
       )
       Y=&Sine(
        Normalized Angle=&Add_Float(A=$Particles.Initial.PartitionParticles.Normalized Index In Partition, B=$Emitter.Age)
        Period=1
        Scale=&RandomRangeFloat(Minimum=-150, Maximum=150)
        Bias=0
       )
       Z=&Cosine(
        Normalized Angle=&Add_Float(A=$Particles.Initial.PartitionParticles.Normalized Index In Partition, B=$Emitter.Age)
        Period=1
        Scale=&RandomRangeFloat(Minimum=-150, Maximum=150)
        Bias=0
       )
      )
      Float=&FloatFromCurve(
       FloatCurve=DataInterface() {propertyValues {Curve {
          defaultValue=1
          keys [
           {tangentMode=RCTM_User, time=0, value=1}
           {arriveTangent=-0.0405, interpMode=RCIM_Cubic, leaveTangent=-0.0405, tangentMode=RCTM_User, time=0.7009, value=1}
           {arriveTangent=-3.8648, interpMode=RCIM_Cubic, leaveTangent=-3.8648, tangentMode=RCTM_User, time=1, value=0.5}
          ]
         }}}
       Scale Curve=1.25
      )
     )
    )) {isSetParameters=true}
   &DynamicMaterialParameters(
    Index 0 Param 1=&Multiply_FloatByInt(Float=-0.3, Integer=$Particles.PartitionParticles.Partition)
    Index 1 Param 2=&FloatFromCurve(FloatCurve=DataInterface() {propertyValues {Curve {defaultValue=1, keys [{tangentMode=RCTM_User, time=0, value=0}, {tangentMode=RCTM_User, time=1, value=1}]}}})
    Module.Index0Float0Write=true
    Module.Index0Float1Write=true
    Module.Index0Float2Write=true
    Module.Index0Float3Write=true
    Module.Index1Float0Write=true
    Module.Index1Float1Write=true
    Module.Index1Float2Write=true
    Module.Index1Float3Write=true
    Write Parameter Index 1=::NewEnumerator1
   )
   &Color(
    Color=&ColorFromCurve(
     ColorCurve=DataInterface() {propertyValues {
       AlphaCurve {keys [<at 0 2 RCIM_Cubic>]}
       BlueCurve {keys [<at 0 1.6054>,<at 0.3734 0.6979>]}
       GreenCurve {keys [<at 0 0>,<at 0.3734 0>]}
       RedCurve {keys [<at 0 2.5>,<at 0.3734 0.166>]}
      }}
     CurveIndex=$Particles.PartitionParticles.Normalized Index In Partition
    )
    Scale Alpha=&FloatFromCurve(FloatCurve=DataInterface() {propertyValues {Curve {
        defaultValue=1
        keys [
         {tangentMode=RCTM_User, time=0, value=1}
         {arriveTangent=-0.0405, interpMode=RCIM_Cubic, leaveTangent=-0.0405, tangentMode=RCTM_User, time=0.7009, value=1}
         {arriveTangent=-8.0089, interpMode=RCIM_Cubic, leaveTangent=-8.0089, tangentMode=RCTM_User, time=1, value=0}
        ]
       }}})
   )
   /SetVariables(
    Particles.GlowColor=&MakeLinearColorFromVectorAndFloat(
     "Vector (RGB)"=&MakeVectorFromLinearColorRGB(Color=$Particles.Color)
     "Float (Alpha)"=&FloatFromCurve(
      FloatCurve=DataInterface() {propertyValues {Curve {
         defaultValue=1
         keys [
          {tangentMode=RCTM_User, time=0, value=1}
          {arriveTangent=-0.0405, interpMode=RCIM_Cubic, leaveTangent=-0.0405, tangentMode=RCTM_User, time=0.7009, value=1}
          {arriveTangent=-8.0089, interpMode=RCIM_Cubic, leaveTangent=-8.0089, tangentMode=RCTM_User, time=1, value=0}
         ]
        }}}
      Scale Curve=0.5
     )
    )
    Particles.GlowWidth:NiagaraFloat=&Multiply_Float(A=$Particles.RibbonWidth, B=1.25)
   ) {isSetParameters=true}
  ]
  rendererClasses [&NiagaraRibbonRendererProperties]
  renderers [
   {rendererClass=&NiagaraRibbonRendererProperties}
   {rendererClass=&NiagaraRibbonRendererProperties, rendererIndex=1}
  ]
 }
 ShellBurst {
  emitterUpdateScript [
   &EmitterState()
   &SpawnBurst_Instantaneous(Use Spawn Probability=false, UseLoopCountLimit=false)
  ]
  particleSpawnScript [
   &InitializeParticle(
    Color=<rgba 1 1 1 0.5>
    Lifetime=$User.Duration
    Mesh Scale Mode=::NewEnumerator3
    Mesh Uniform Scale=26.2282
    Position=<15 0 5>
    Position Mode=::NewEnumerator1
    Sprite Size=&Multiply_Vector2D_ByFloat(Float=1.4, Vector2D=<815 500>)
    Sprite Size Mode=::NewEnumerator4
    UsePositionOffset=false
    Write Material Random=true
   )
   &CameraOffset(Camera Offset Amount=300)
  ]
  particleUpdateScript [
   &ParticleState()
   &ScaleMeshSize(Scale Factor=&VectorFromFloat(Value=&FloatFromCurve(FloatCurve=DataInterface() {propertyValues {Curve {keys [<at 0 0.8>,<at 1 1>]}}}, Scale Curve=10)))
   &DynamicMaterialParameters(
    Index 0 Param 1=0.25
    Index 0 Param 2=-1
    Index 0 Param 3=1
    Index 0 Param 4=1
    Module.Index0Param0Write=true
    Module.Index0Param1Write=true
    Module.Index0Param2Write=true
    Module.Index0Param3Write=true
   )
   &ScaleColor(
    Scale Alpha=&FloatFromCurve(FloatCurve=DataInterface() {propertyValues {Curve {keys [<at 0.0999 1.0037 RCIM_Cubic>,<at 1 1>]}}})
    ScaleA=true
    ScaleRGB=true
   )
   /SetVariables(Particles.Position=&AddVectorToPosition(
     Position=&Lerp_Vector(A=$System.SampleSkeletalMesh.SampledPosition, B=$System.SampleSkeletalMesh001.SampledPosition, Alpha=0)
     Vector=&TransformVector(Vector=&Multiply_VectorByFloat(Vector=<20 0 10>, Float=1), Destination Space=::Simulation)
    )) {enabled=false, isSetParameters=true, module=SetVariables_#13}
   &SpriteFacingAndAlignment(
    Alignment Coordinate Space=::World
    Sprite Alignment=&NormalizeVector(
     Vector To Normalize=&SubtractVectorFromPosition(Position=$Engine.Owner.Position, Vector=Linked() {linkedVariable {name=User.BeamEnd_0, type=NiagaraPosition}})
     Safe Normalize=true
     Safe Fallback Threshold=0
     Safe Fallback Vector=<0 0 0.001>
    )
    WriteAlignment=true
    WriteFacing=false
   )
   &ScaleColor(
    Scale Alpha=&FloatFromCurve(
     FloatCurve=DataInterface() {propertyValues {Curve {keys [{interpMode=RCIM_Cubic, tangentMode=RCTM_Break, time=0, value=0}, <at 0.3 1>]}}}
     CurveIndex=$Engine.System.Age
    )
    ScaleA=true
    ScaleRGB=true
   ) {module=ScaleColor001}
   &ScaleColor(
    Scale Alpha=&FloatFromCurve(
     FloatCurve=DataInterface() {propertyValues {Curve {keys [{interpMode=RCIM_Cubic, leaveTangent=-3.7201, tangentMode=RCTM_Break, time=0, value=1}, <at 1 0>]}}}
     CurveIndex=$System.TimeForDis
    )
    ScaleA=true
    ScaleRGB=true
   ) {module=ScaleColor002}
  ]
  rendererClasses [&NiagaraSpriteRendererProperties]
  renderers [{rendererClass=&NiagaraSpriteRendererProperties}]
 }
 ShellBurst001 {
  emitterUpdateScript [
   &EmitterState()
   &SpawnBurst_Instantaneous(Use Spawn Probability=false, UseLoopCountLimit=false)
  ]
  particleSpawnScript [
   &InitializeParticle(
    Lifetime=$User.Duration
    Mesh Scale Mode=::NewEnumerator3
    Mesh Uniform Scale=26.2282
    Position=<15 0 5>
    Position Mode=::NewEnumerator1
    Sprite Size=<200 1700>
    Sprite Size Mode=::NewEnumerator4
    UsePositionOffset=false
    Write Material Random=true
   )
   &CameraOffset(Camera Offset Amount=300)
  ]
  particleUpdateScript [
   &ParticleState()
   &ScaleMeshSize(Scale Factor=&VectorFromFloat(Value=&FloatFromCurve(FloatCurve=DataInterface() {propertyValues {Curve {keys [<at 0 0.8>,<at 1 1>]}}}, Scale Curve=10)))
   &DynamicMaterialParameters(
    Index 0 Param 1=0.25
    Index 0 Param 2=-1
    Index 0 Param 3=1
    Index 0 Param 4=1
    Module.Index0Param0Write=true
    Module.Index0Param1Write=true
    Module.Index0Param2Write=true
    Module.Index0Param3Write=true
   )
   &ScaleColor(ScaleA=true, ScaleRGB=true)
   /SetVariables(Particles.Position=&AddVectorToPosition(
     Position=&Lerp_Vector(A=$System.SampleSkeletalMesh.SampledPosition, B=$System.SampleSkeletalMesh001.SampledPosition, Alpha=0)
     Vector=&TransformVector(Vector=&Multiply_VectorByFloat(Vector=<20 0 10>, Float=1), Destination Space=::Simulation)
    )) {enabled=false, isSetParameters=true, module=SetVariables_#13}
   &SpriteFacingAndAlignment(
    Alignment Coordinate Space=::World
    Sprite Alignment=&NormalizeVector(
     Vector To Normalize=&SubtractVectorFromPosition(Position=$User.BeamEnd_0, Vector=Linked() {linkedVariable {name=Engine.Owner.Position, type=NiagaraPosition}})
     Safe Normalize=true
     Safe Fallback Threshold=0
     Safe Fallback Vector=<0 0 0.001>
    )
    WriteAlignment=true
    WriteFacing=false
   )
   &ScaleColor(
    Scale Alpha=&FloatFromCurve(
     FloatCurve=DataInterface() {propertyValues {Curve {keys [{interpMode=RCIM_Cubic, tangentMode=RCTM_Break, time=0, value=0}, <at 0.3 1>]}}}
     CurveIndex=$Engine.System.Age
    )
    ScaleA=true
    ScaleRGB=true
   ) {module=ScaleColor001}
   &ScaleColor(
    Scale Alpha=&FloatFromCurve(
     FloatCurve=DataInterface() {propertyValues {Curve {keys [{interpMode=RCIM_Cubic, leaveTangent=-3.7201, tangentMode=RCTM_Break, time=0, value=1}, <at 1 0>]}}}
     CurveIndex=$System.TimeForDis
    )
    ScaleA=true
    ScaleRGB=true
   ) {module=ScaleColor002}
  ]
  rendererClasses [&NiagaraSpriteRendererProperties]
  renderers [{rendererClass=&NiagaraSpriteRendererProperties}]
 }
}
summary {
 systemName=NS_PistolCanary_MM_Attack_Beam
 emitters [
  {emitterName=Muzzle_Inner_Flares, rendererClasses [&NiagaraSpriteRendererProperties]}
  {emitterName=CoreFlare001, rendererClasses [&NiagaraSpriteRendererProperties]}
  {emitterName=Beams, rendererClasses [&NiagaraRibbonRendererProperties]}
  {emitterName=ShellBurst, rendererClasses [&NiagaraSpriteRendererProperties]}
  {emitterName=ShellBurst001, rendererClasses [&NiagaraSpriteRendererProperties]}
  {emitterName=Minimal, rendererClasses [&NiagaraSpriteRendererProperties]}
  {emitterName=Minimal002, rendererClasses [&NiagaraSpriteRendererProperties]}
  {emitterName=Impact_Flare, rendererClasses [&NiagaraSpriteRendererProperties, &NiagaraLightRendererProperties]}
  {emitterName=Energy_Sphere001, rendererClasses [&NiagaraMeshRendererProperties]}
  {emitterName=NE_EnergyRibbons, rendererClasses [&NiagaraRibbonRendererProperties]}
 ]
 userVariables [
  User.Beam End {type=NiagaraPosition, description="", defaultValue {}}
  User.Beam_End {type=NiagaraPosition, description="", defaultValue {}}
  User.BeamEnd_0 {type=NiagaraPosition, description="", defaultValue {}}
  User.BeamWidth {type=NiagaraFloat, defaultValue=1, description=""}
  User.bHitReaction {type=NiagaraBool}
  User.bSpawnWater {type=NiagaraBool}
  User.Delay {type=NiagaraFloat, defaultValue=0, description=""}
  User.DisappearTime {type=NiagaraFloat, defaultValue=0.8, description=""}
  User.Duration {type=NiagaraFloat, defaultValue=3.8, description=""}
  User.EnergyBeamSwipe {type=NiagaraFloat, defaultValue=0.456, description=""}
  User.Initial Start Time {type=NiagaraFloat, defaultValue=0, description=""}
  User.InterpolateRateOfChange {type=NiagaraFloat, defaultValue=20, description=""}
  User.OverallOpacity {type=NiagaraFloat, defaultValue=1, description=""}
  User.PrimaryColor {type=LinearColor, description="", defaultValue {struct=LinearColor, value=<rgba 5 0.2882 0.6068 1>}}
  User.SecondaryColor {type=LinearColor, description="", defaultValue {struct=LinearColor, value=<rgba 0.7101 0.0959 0.1594 5>}}
  User.Size {type=NiagaraFloat, defaultValue=1, description=""}
  User.Stop Location {type=NiagaraPosition, description="", defaultValue {}}
 ]
}
